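; Off-road tires vehicle mod box, a child Blueprint actor that shows an "OFF ROAD TIRES" world label with its mod icon.
; BeginPlay calls the parent, then sets the label text and icon on the world UI.
; When the box dies, it casts the damage causer to a Fort Athena vehicle and repairs all that vehicle's damageable parts.
; The death event also forwards to the parent class, and no tire-change logic appears here.
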(event EventBeginPlay
  (|CallParentFunction)
  (WorldUI|SetLabelText "NSLOCTEXT(\"[E8DB4E1EF9D5B418D9B8A5845AEF3DC8]\", \"4C1738E14733ED0736D845B53E3A5C17\", \"OFF ROAD TIRES\")" "" (Variables|Default|GetModIcon)))

(event EventOnDied_Event
  (params (DamagedActor :type Actor%20Object%20Reference) (Damage :type Float%20%28single-precision%29) (InstigatedBy :type Controller%20Object%20Reference) (DamageCauser :type Actor%20Object%20Reference) (HitLocation :type Vector :default "0, 0, 0") (FHitComponent :type Primitive%20Component%20Object%20Reference) (BoneName :type Name :default "None") (Momentum :type Vector :default "0, 0, 0"))
  (bind _asfort_athena_vehicle (Utilities|Casting|CastToFortAthenaVehicle DamageCauser)
    (:then
      (Class|FortAthenaVehicle|RepairAllDamageableParts _asfort_athena_vehicle 1.0))
    (:CastFailed
      _))
  (|CallParentFunction DamagedActor Damage InstigatedBy DamageCauser HitLocation FHitComponent BoneName Momentum))

(fn ConstructionScript
  (|CallParentFunction))
